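# SoundCue for the looping float sound of Nug, built from a looping falling-wind in-air wave.
# An Enveloper ramps volume up from silence, then a Modulator randomizes pitch upward before glider close-detail attenuation is applied.
# The cue plays quietly in the in-air glider sound class and overrides concurrency to cap simultaneous instances.
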SoundCue /Irwin/Sounds/Nug/Nug_Float_Loop_Cue
root=Attenuation_1
nodes [
 {
  id=Attenuation_1
  children [Modulator_0]
  properties {attenuationSettings=/Game/Sounds/Attenuation/Fort_Player_Glider_CloseDetail_Athena_Att}
 }
 {
  id=Enveloper_0
  children [WavePlayer_2]
  properties {pitchCurve {editorCurveData {keys [{time=0}]}}, volumeCurve {editorCurveData {keys [<at 0 0>,<at 1 1>]}}}
 }
 {id=Modulator_0, children [Enveloper_0], properties {pitchMax=2, pitchMin=1.5}}
 {id=WavePlayer_2, properties {bLooping=true, soundWaveAssetPtr=/Game/Athena/Sounds/Wind_Falling/Character_InAir_AudioTell_01}}
]
properties {
 bOverrideConcurrency=true
 maxDistance=6000
 soundClassObject=&Fort_Gliders_InAir
 volumeMultiplier=0.2
 concurrencyOverrides {maxCount=2}
}
